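# Masked parent material for the TadPole Infinity wheel on skeletal meshes, with glowing infinity-depth tire tread and rim shapes.
# MF_Infinity_OffsetFunction offsets tread and shape textures for depth, feeding emissive patterns alongside a panning gradient and Fresnel term.
# Vertex color and RGB mask textures supply channels for blending tire and rim layers.
# Static switches toggle a camera-dependent hue-shifted custom color (MultiEmissiveBlend) and a masked-or-constant LOD branch (LOds).
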Material /VehicleCosmetics/Wheels/TadPole_Infinity/Materials/M_WheelParent_TadPoleInfinity
graph {
 properties {bAutomaticallySetUsageInEditor=false, bUsedWithSkeletalMesh=true, blendMode=BLEND_Masked, usedShadingModels=Default Lit}
 nodes [
  LinearInterpolate(A=Multiply(A=NamedRerouteUsage(), B=NamedRerouteUsage()), B=Constant(), Alpha=NamedRerouteUsage())
  n1 = VectorParameter() {parameterName=CustomColor, defaultValue=<rgba 0.7682 0.6308 0.0382 1>}
  n2 = ScalarParameter() {defaultValue=-0.2, parameterName=ColorAngle}
  n3 = CameraVectorWS()
  n4 = Multiply(
   A=Multiply(
    A=MaterialFunctionCall!(Color=@n1.RGBA, Angle=@n2) {materialFunction=/VehicleCosmetics/Wheels/Concept_Inverted/Materials/MF_HUE_Shift}.Result
    B=Constant() {r=1}
   )
   B=DotProduct(A=Constant() {r=0.1}, B=@n3)
  )
  n5 = Multiply(A=DotProduct(A=Constant() {r=0.2}, B=OneMinus(None=@n3)), B=@n1.RGBA)
  NamedRerouteDeclaration(None=StaticSwitchParameter(True=Add(A=@n4, B=@n5), False=@n1.RGBA) {parameterName=MultiEmissiveBlend}) {name=HueShift, nodeColor=<rgba 0.7812 1 0 1>}
  n6 = ScalarParameter() {defaultValue=2, parameterName=LightEmissivePower}
  n7 = MaterialFunctionCall!(
   Texture=TextureObject() {samplerType=SAMPLERTYPE_LinearGrayscale, texture=/VehicleCosmetics/Wheels/TadPole_Infinity/Textures/T_TadpoleInfinity_Tire_M}
   Offset=ScalarParameter() {defaultValue=-1, parameterName=BumpOffsetAmountTread}
   UVs
   RotateUVs
   GradientFalloff
   GradientDarkness=ScalarParameter() {defaultValue=0.001, parameterName=GradientDarkness}
  ) {materialFunction=/VehicleCosmetics/Wheels/ClockGear_SE/Materials/MF_Infinity_OffsetFunction}
  n8 = Multiply(
   A=Multiply(A=ComponentMask(None=@n7.Result) {g=true}, B=NamedRerouteUsage())
   B=ScalarParameter() {defaultValue=5, parameterName=OriginalyPowerTread}
  )
  n9 = ScalarParameter() {defaultValue=0.5, parameterName=Initial Tread Brightness}
  n10 = Multiply(A=@n9, B=@n8)
  n11 = Multiply(A=NamedRerouteUsage(), B=Constant() {r=1})
  n12 = Add(
   A=Multiply(A=@n8, B=@n9)
   B=Multiply(A=Multiply(A=@n10, B=@n11), B=ScalarParameter() {defaultValue=20, parameterName=InfinityPulseGlowPower})
  )
  NamedRerouteDeclaration(None=Multiply(A=Multiply(A=@n6, B=@n12), B=NamedRerouteUsage())) {name=TirePattern, nodeColor=<rgba 0.8984 0 1 1>}
  n13 = VertexColor()
  NamedRerouteDeclaration(None=@n13.R) {name=VC_R_CH, nodeColor=<rgba 1 0 0.0912 1>}
  n14 = TextureSampleParameter2D() {parameterName=Normal_Rim, samplerType=SAMPLERTYPE_Normal, texture=/VehicleCosmetics/Wheels/TadPole_Infinity/Textures/T_TadPoleInfinity_Rim_N}
  NamedRerouteDeclaration(None=@n14.RGB) {name=Normal_Rim, nodeColor=<rgba 1 0 0.0469 1>}
  n15 = Constant() {r=2}
  n16 = DotProduct(A=NamedRerouteUsage(), B=CameraVectorWS())
  n17 = OneMinus(None=@n16)
  n18 = Multiply(A=Clamp(None=Multiply(A=@n15, B=@n16)) {minDefault=0.35}, B=Clamp(None=ConstantBiasScale(None=@n17) {bias=-0.375, scale=1.666}))
  NamedRerouteDeclaration(None=Multiply(A=Multiply(A=@n18, B=@n18), B=Constant() {r=2})) {name=Fresnel, nodeColor=<rgba 0 0.3672 1 1>}
  NamedRerouteDeclaration(None=TextureSample() {samplerType=SAMPLERTYPE_Normal, texture=/VehicleCosmetics/Wheels/TadPole_Infinity/Textures/T_TadpoleInfinity_Tire_N}.RGB) {name=Normal_Tire, nodeColor=<rgba 0 0.2006 1 1>}
  NamedRerouteDeclaration(None=@n13.G) {name=VC_G_CH, nodeColor=<rgba 0 1 0.2734 1>}
  n19 = TextureSampleParameter2D() {
   parameterName=TreadRGB
   samplerType=SAMPLERTYPE_LinearColor
   texture=/VehicleCosmetics/Wheels/TadPole_Infinity/Textures/T_TadpoleInfinity_Tire_RGB
  }
  NamedRerouteDeclaration(None=@n19.G) {name=Tread_Mask_Green_CH, nodeColor=<rgba 0.0105 1 0 1>}
  NamedRerouteDeclaration(None=@n19.B) {name=Tread_Mask_Blue_CH, nodeColor=<rgba 0 0.0625 1 1>}
  n20 = TextureSample(UVs=Panner() {speedX=0.2}) {
   samplerSource=SSM_Wrap_WorldGroupSettings
   samplerType=SAMPLERTYPE_LinearColor
   texture=/VehicleCosmetics/Wheels/TadPole_Infinity/Textures/BaseGradient_Vert
  }
  NamedRerouteDeclaration(None=Multiply(A=@n20.RGBA, B=Constant() {r=1.25})) {name=Animation, nodeColor=<rgba 0.1953 0 1 1>}
  n21 = TextureSampleParameter2D() {
   parameterName=Rim_Mask
   samplerType=SAMPLERTYPE_LinearColor
   texture=/VehicleCosmetics/Wheels/TadPole_Infinity/Textures/T_TadPoleInfinity_Rim_RGB
  }
  NamedRerouteDeclaration(None=@n21.G) {name=Rim_Mask_Green_CH, nodeColor=<rgba 0.125 1 0 1>}
  NamedRerouteDeclaration(None=@n21.B) {name=Rim_Mask_Blue_CH_1, nodeColor=<rgba 0 0.0421 1 1>}
  NamedRerouteDeclaration(None=@n21.R) {name=Rim_Mask_Red_CH_2, nodeColor=<rgba 1 0.0106 0 1>}
  NamedRerouteDeclaration(None=@n6) {name=Light_Emissive, nodeColor=<rgba 1 0.9141 0 1>}
  n22 = ComponentMask(None=MaterialFunctionCall(
    Texture=TextureObject() {samplerType=SAMPLERTYPE_LinearColor, texture=/VehicleCosmetics/Wheels/TadPole_Infinity/Textures/Wheel_Tadpole_InfinityShapes}
    Offset=ScalarParameter() {defaultValue=-1, parameterName=BumpOffsetAmountTread}
    GradientDarkness=ScalarParameter() {defaultValue=0.001, parameterName=GradientDarkness}
   ) {materialFunction=&MF_Infinity_OffsetFunction}.Result) {g=true}
  n23 = Multiply(A=@n22, B=ScalarParameter() {defaultValue=6, parameterName=OriginalyPowerInfinity})
  n24 = NamedRerouteUsage()
  n25 = LinearInterpolate(A=NamedRerouteUsage(), B=@n22, Alpha=NamedRerouteUsage())
  n26 = LinearInterpolate(A=@n25, B=Add(A=Multiply(A=@n23, B=@n24), B=@n25), Alpha=NamedRerouteUsage())
  n27 = NamedRerouteUsage()
  NamedRerouteDeclaration(None=Multiply(A=Multiply(A=@n26, B=@n27), B=NamedRerouteUsage())) {name=Rim, nodeColor=<rgba 0.5938 0 1 1>}
  n28 = Multiply(A=NamedRerouteUsage(), B=NamedRerouteUsage())
  n29 = NamedRerouteUsage()
  n30 = ScalarParameter() {defaultValue=0.4, parameterName=EmissiveTreadPower}
  n31 = Multiply(A=Add(A=Multiply(A=@n28, B=@n29), B=Multiply(A=@n28, B=@n30)), B=ScalarParameter() {defaultValue=1, parameterName=EmissivePowerr})
  n32 = NamedRerouteUsage()
  n33 = NamedRerouteUsage()
  Reroute(None=LinearInterpolate(A=NamedRerouteUsage(), B=LinearInterpolate(A=@n31, B=@n32, Alpha=@n33), Alpha=NamedRerouteUsage()))
  Reroute(None=LinearInterpolate(A=NamedRerouteUsage(), B=NamedRerouteUsage(), Alpha=NamedRerouteUsage()))
  ScalarParameter() {defaultValue=1, parameterName=Roughness}
  ScalarParameter() {parameterName=Metallic}
  n34 = Constant() {r=1}
  n35 = NamedRerouteUsage()
  n36 = NamedRerouteUsage()
  StaticSwitchParameter(True=Add(A=LinearInterpolate(A=@n34, B=@n35, Alpha=@n36), B=NamedRerouteUsage()), False=Constant() {r=1}) {parameterName=LOds}
 ]
}
